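# Sound cue for the Octoberfest cuckoo back bling sound in Athena.
# A single wave player feeds the output directly with one cuckoo sound wave, without randomization or modulation.
# It borrows the close-range pets attenuation to spatialize the sound and is routed through the back blings sound class.
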SoundCue /Game/Athena/Sounds/BackBlings/BackBling_Octoberfest_Cuckoo_Cue
root=WavePlayer_0
nodes [{id=WavePlayer_0, properties {soundWaveAssetPtr=&/Octoberfest_BackBling_Cuckoo_01}}]
properties {
 attenuationSettings=&Fort_Pets_Close_Athena_Att
 maxDistance=1800.0001
 soundClassObject=/Game/Sounds/FortSoundClassesAndModes/Fort_Backblings
 volumeMultiplier=1
}
