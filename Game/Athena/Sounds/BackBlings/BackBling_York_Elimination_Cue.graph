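# Sound cue for the York back bling, named as its elimination sound.
# A single wave player feeds the Backpack_York wave straight to the output at reduced volume, routed through the back bling sound class with Athena emote attenuation.
# Overridden concurrency limits it to one playing instance per owner.
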SoundCue /Game/Athena/Sounds/BackBlings/BackBling_York_Elimination_Cue
root=WavePlayer_1
nodes [{id=WavePlayer_1, properties {soundWaveAssetPtr=&/Backpack_York}}]
properties {
 attenuationSettings=&Fort_Emotes_Athena_Att
 bOverrideConcurrency=true
 maxDistance=1600
 soundClassObject=/Game/Sounds/FortSoundClassesAndModes/Fort_Backblings
 volumeMultiplier=0.3
 concurrencyOverrides {bLimitToOwner=true, maxCount=1, volumeScale=0.7}
}
